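# GrappleKnife_Bounce_Cue is a Sound Cue in the GrappleWeapon assets, named for the grapple knife's bounce sound.
# Its node graph is empty with no root node, so it references no audio and plays nothing as it stands.
# It still carries playback setup: the GrappleKnife_Att attenuation asset, the Fort_Gadgets_Items sound class, and a concurrency override that prevents new instances once the limit is reached.
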SoundCue /GrappleWeapon/Assets/Audio/GrappleKnife_Bounce_Cue
root=null
nodes []
properties {
 attenuationSettings=&/GrappleKnife_Att
 bOverrideConcurrency=true
 maxDistance=12500
 soundClassObject=&Fort_Gadgets_Items
 volumeMultiplier=1
 concurrencyOverrides {resolutionRule=PreventNew}
}
