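# Sound Cue for the GrappleWeapon's grapple knife bounce, in its "Player" variant.
# A Random node picks one of five equally weighted bounce waves and feeds a Modulator at default settings as the root.
# It uses the GrappleKnife foley attenuation and the gadgets/items sound class.
# A concurrency override prevents new instances once the limit is reached, rather than cutting off playing ones.
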SoundCue /GrappleWeapon/Assets/Audio/GrappleKnife_Bounce_Player_Cue
root=Modulator_0
nodes [
 {id=Modulator_0, children [Random_1]}
 {
  id=Random_1
  children [WavePlayer_4,WavePlayer_5,WavePlayer_6,WavePlayer_7,WavePlayer_8]
  properties {bShouldExcludeFromBranchCulling=true, weights [1,1,1,1,1]}
 }
 {id=WavePlayer_4, properties {soundWaveAssetPtr=&/GrappleKnife_Player_Bounce_01}}
 {id=WavePlayer_5, properties {soundWaveAssetPtr=&/GrappleKnife_Player_Bounce_02}}
 {id=WavePlayer_6, properties {soundWaveAssetPtr=&/GrappleKnife_Player_Bounce_03}}
 {id=WavePlayer_7, properties {soundWaveAssetPtr=&/GrappleKnife_Player_Bounce_04}}
 {id=WavePlayer_8, properties {soundWaveAssetPtr=&/GrappleKnife_Player_Bounce_05}}
]
properties {
 attenuationSettings=&/GrappleKnife_Foley_Att
 bOverrideConcurrency=true
 maxDistance=6000
 soundClassObject=&Fort_Gadgets_Items
 volumeMultiplier=1
 concurrencyOverrides {resolutionRule=PreventNew}
}
