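# Sound cue for the coil sound when the grapple knife is cast, in the GrappleWeapon gadget.
# A Random node picks one of four equally weighted coil wave variations, which feeds a Modulator that varies volume and pitch per play.
# It uses the grapple knife attenuation and the gadgets/items sound class, and overrides concurrency with a low instance cap.
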SoundCue /GrappleWeapon/Assets/Audio/GrappleKnife_Cast_Coil_Cue
root=Modulator_1
nodes [
 {id=Modulator_1, children [Random_1], properties {pitchMax=1.2, volumeMax=1, volumeMin=0.75}}
 {
  id=Random_1
  children [WavePlayer_4,WavePlayer_5,WavePlayer_6,WavePlayer_7]
  properties {bShouldExcludeFromBranchCulling=true, weights [1,1,1,1]}
 }
 {id=WavePlayer_4, properties {soundWaveAssetPtr=&/GrappleKnife_Cast_Coil_01}}
 {id=WavePlayer_5, properties {soundWaveAssetPtr=&/GrappleKnife_Cast_Coil_02}}
 {id=WavePlayer_6, properties {soundWaveAssetPtr=&/GrappleKnife_Cast_Coil_03}}
 {id=WavePlayer_7, properties {soundWaveAssetPtr=&/GrappleKnife_Cast_Coil_04}}
]
properties {
 attenuationSettings=&/GrappleKnife_Att
 bOverrideConcurrency=true
 maxDistance=10000
 soundClassObject=&Fort_Gadgets_Items
 volumeMultiplier=0.7
 concurrencyOverrides {maxCount=3}
}
