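# GrappleKnife_Cast_Release_Cue is a sound cue for the grapple knife's cast release, in the gadget items sound class.
# Its graph has no root node and no sound nodes, so it holds only playback settings.
# It uses the grapple knife foley attenuation with a set maximum distance and overrides concurrency to cap simultaneous instances.
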SoundCue /GrappleWeapon/Assets/Audio/GrappleKnife_Cast_Release_Cue
root=null
nodes []
properties {
 attenuationSettings=&/GrappleKnife_Foley_Att
 bOverrideConcurrency=true
 maxDistance=6000
 soundClassObject=&Fort_Gadgets_Items
 concurrencyOverrides {maxCount=3}
}
