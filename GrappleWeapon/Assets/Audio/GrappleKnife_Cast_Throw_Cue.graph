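# Sound cue for the GrappleWeapon grapple knife, named for the cast throw but built from four projectile reel-in wave variants.
# A Random node picks one variant at equal weight and feeds a Modulator that varies pitch and volume on each play.
# Grapple knife attenuation spatializes it under the gadgets and items sound class, and a concurrency override caps simultaneous instances at a small count.
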SoundCue /GrappleWeapon/Assets/Audio/GrappleKnife_Cast_Throw_Cue
root=Modulator_0
nodes [
 {id=Modulator_0, children [Random_0], properties {pitchMax=1.2, volumeMax=1, volumeMin=0.8}}
 {
  id=Random_0
  children [WavePlayer_0,WavePlayer_1,WavePlayer_2,WavePlayer_3]
  properties {bShouldExcludeFromBranchCulling=true, weights [1,1,1,1]}
 }
 {id=WavePlayer_0, properties {soundWaveAssetPtr=&/GrappleKnife_Prj_ReelIn_01}}
 {id=WavePlayer_1, properties {soundWaveAssetPtr=&/GrappleKnife_Prj_ReelIn_02}}
 {id=WavePlayer_2, properties {soundWaveAssetPtr=&/GrappleKnife_Prj_ReelIn_03}}
 {id=WavePlayer_3, properties {soundWaveAssetPtr=&/GrappleKnife_Prj_ReelIn_04}}
]
properties {
 attenuationSettings=&/GrappleKnife_Att
 bOverrideConcurrency=true
 maxDistance=10000
 soundClassObject=&Fort_Gadgets_Items
 volumeMultiplier=0.9
 concurrencyOverrides {maxCount=3}
}
