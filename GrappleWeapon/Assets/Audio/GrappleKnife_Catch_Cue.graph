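# Sound cue for the grapple knife catch sound of the grapple weapon, routed to the weapon foley sound class.
# A Random node picks one of four equally weighted catch waves, and a Modulator varies its pitch and volume on each play.
# It uses weapon reload attenuation, overrides concurrency, and has the debug flag turned on.
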SoundCue /GrappleWeapon/Assets/Audio/GrappleKnife_Catch_Cue
root=Modulator_0
nodes [
 {id=Modulator_0, children [Random_1], properties {pitchMax=1.25, pitchMin=0.9, volumeMax=1, volumeMin=0.5}}
 {
  id=Random_1
  children [WavePlayer_3,WavePlayer_4,WavePlayer_5,WavePlayer_6]
  properties {bShouldExcludeFromBranchCulling=true, weights [1,1,1,1]}
 }
 {id=WavePlayer_3, properties {soundWaveAssetPtr=&/GrappleKnife_Catch_01}}
 {id=WavePlayer_4, properties {soundWaveAssetPtr=&/GrappleKnife_Catch_02}}
 {id=WavePlayer_5, properties {soundWaveAssetPtr=&/GrappleKnife_Catch_03}}
 {id=WavePlayer_6, properties {soundWaveAssetPtr=&/GrappleKnife_Catch_04}}
]
properties {
 attenuationSettings=&Fort_Weapon_Reload_Att
 bDebug=true
 bOverrideConcurrency=true
 maxDistance=2250
 soundClassObject=&Fort_WeaponFoley
 volumeMultiplier=0.601
}
