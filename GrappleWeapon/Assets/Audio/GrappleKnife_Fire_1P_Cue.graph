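# First-person fire sound cue for the grapple knife weapon, with an empty node graph and no root, so it references no sound waves.
# It carries only playback settings: a first-person weapon projectile attenuation, the first-person ranged-fire sound class and the first-person gunfire concurrency set.
# It also stores attenuation and concurrency override values, the latter using a prevent-new resolution rule.
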SoundCue /GrappleWeapon/Assets/Audio/GrappleKnife_Fire_1P_Cue
root=null
nodes []
properties {
 attenuationSettings=&Fort_Weapon_Projectile_FirstPerson_Athena_Att
 maxDistance=2097152
 soundClassObject=&Fort_WeaponRanged_Fire_1P_PSM
 volumeMultiplier=1
 attenuationOverrides {falloffDistance=5000, attenuationShapeExtents {x=1000}}
 concurrencyOverrides {resolutionRule=PreventNew}
 concurrencySet [&Weapon_Gunfire_FirstPerson_Conc]
}
