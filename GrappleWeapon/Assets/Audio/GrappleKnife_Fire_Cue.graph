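# GrappleKnife_Fire_Cue is a SoundCue in the GrappleWeapon content, named for the grapple knife's fire sound.
# It has no root or sound nodes, so as it stands it plays no audio.
# It references the local and global weapon gunfire concurrency sets and carries attenuation and concurrency override settings.
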SoundCue /GrappleWeapon/Assets/Audio/GrappleKnife_Fire_Cue
root=null
nodes []
properties {
 maxDistance=2097152
 volumeMultiplier=1
 attenuationOverrides {falloffDistance=5000, attenuationShapeExtents {x=1000}}
 concurrencyOverrides {resolutionRule=PreventNew}
 concurrencySet [&Weapon_Gunfire_Local_Conc, &Weapon_Gunfire_Global_Conc]
}
