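# Sound cue for the grapple knife's default impact, judging by its name and wave assets.
# A Random node picks among five equally weighted impact waves, and a Modulator varies pitch and slightly varies volume on each play.
# The cue uses small melee weapon attenuation and the impacts sound class, with a concurrency override that caps simultaneous instances despite an assigned melee concurrency set.
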SoundCue /GrappleWeapon/Assets/Audio/GrappleKnife_Hit_Default_Cue
root=Modulator_4
nodes [
 {id=Modulator_4, children [Random_0], properties {pitchMax=1.2, volumeMax=1, volumeMin=0.9}}
 {
  id=Random_0
  children [WavePlayer_1, WavePlayer_2, WavePlayer_4, WavePlayer_11, WavePlayer_14]
  properties {bShouldExcludeFromBranchCulling=true, weights [1,1,1,1,1]}
 }
 {id=WavePlayer_1, properties {soundWaveAssetPtr=&/GrappleKnife_Prj_Impact_Default_01}}
 {id=WavePlayer_11, properties {soundWaveAssetPtr=&/GrappleKnife_Prj_Impact_Default_04}}
 {id=WavePlayer_14, properties {soundWaveAssetPtr=&/GrappleKnife_Prj_Impact_Default_05}}
 {id=WavePlayer_2, properties {soundWaveAssetPtr=&/GrappleKnife_Prj_Impact_Default_02}}
 {id=WavePlayer_4, properties {soundWaveAssetPtr=&/GrappleKnife_Prj_Impact_Default_03}}
]
properties {
 attenuationSettings=&Fort_Weapon_Melee_Sml_Att
 bOverrideConcurrency=true
 maxDistance=2800
 soundClassObject=&Fort_Impacts
 volumeMultiplier=2.1
 concurrencyOverrides {maxCount=2}
 concurrencySet [&Player_Melee_Weapon_Conc]
}
