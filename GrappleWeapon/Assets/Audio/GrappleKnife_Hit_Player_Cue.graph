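# Sound cue for the grapple knife projectile hitting a player.
# An IsLocal branch plays a random pick of three impact waves with pitch modulation when true, or a quieter, higher-pitched random pick when false or unset.
# It routes to the impacts sound class, uses pickaxe player-impact attenuation, and overrides concurrency with a small per-owner voice limit.
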SoundCue /GrappleWeapon/Assets/Audio/GrappleKnife_Hit_Player_Cue
root=Branch_0
nodes [
 {id=Branch_0, children [Modulator_0,Mixer_0,Mixer_0], properties {boolParameterName=IsLocal}}
 {id=Mixer_0, children [Modulator_1], properties {inputVolume [0.6]}}
 {id=Modulator_0, children [Random_1], properties {pitchMax=1.5, pitchMin=1, volumeMax=1, volumeMin=1}}
 {id=Modulator_1, children [Random_0], properties {pitchMax=1.85, pitchMin=1.5, volumeMax=1, volumeMin=0.8}}
 {id=Random_0, children [WavePlayer_3,WavePlayer_4,WavePlayer_5], properties {bShouldExcludeFromBranchCulling=true, weights [1,1,1]}}
 {id=Random_1, children [WavePlayer_0,WavePlayer_1,WavePlayer_2], properties {bShouldExcludeFromBranchCulling=true, weights [1,1,1]}}
 {id=WavePlayer_0, properties {soundWaveAssetPtr=&/GrappleKnife_Prj_Impact_Player_01}}
 {id=WavePlayer_1, properties {soundWaveAssetPtr=&/GrappleKnife_Prj_Impact_Player_02}}
 {id=WavePlayer_2, properties {soundWaveAssetPtr=&/GrappleKnife_Prj_Impact_Player_03}}
 {id=WavePlayer_3, properties {soundWaveAssetPtr=&GrappleKnife_Prj_Impact_Player_01}}
 {id=WavePlayer_4, properties {soundWaveAssetPtr=&GrappleKnife_Prj_Impact_Player_02}}
 {id=WavePlayer_5, properties {soundWaveAssetPtr=&GrappleKnife_Prj_Impact_Player_03}}
]
properties {
 attenuationSettings=/Game/Sounds/Attenuation/Fort_Weapon_Pickaxe_Player_Impact_Att
 bOverrideConcurrency=true
 maxDistance=3512
 soundClassObject=&Fort_Impacts
 volumeMultiplier=1.575
 attenuationOverrides {
  bAttenuate=false
  bAttenuateWithLPF=true
  distanceAlgorithm=Logarithmic
  falloffDistance=800
  lPFRadiusMax=800
  lPFRadiusMin=400
  attenuationShapeExtents {x=300}
 }
 concurrencyOverrides {bLimitToOwner=true, maxCount=3}
 concurrencySet [/Game/Sounds/Concurrency_Rules/OneMaximum_Owner_StopOldest]
}
