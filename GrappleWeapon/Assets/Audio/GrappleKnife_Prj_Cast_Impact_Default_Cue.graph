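# Sound cue for the grapple knife projectile's cast impact, playing a latch-and-pull sound in the gadgets sound class.
# A Random node picks among three equally weighted wave variants and is excluded from branch culling, feeding a Modulator that varies pitch and volume.
# It references a latch-and-reel attenuation asset.
# An inline concurrency override caps simultaneous instances, taking precedence over the assigned rifle bullet impact concurrency set.
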SoundCue /GrappleWeapon/Assets/Audio/GrappleKnife_Prj_Cast_Impact_Default_Cue
root=Modulator_0
nodes [
 {id=Modulator_0, children [Random_0], properties {pitchMax=1.1, volumeMax=1, volumeMin=0.85}}
 {id=Random_0, children [WavePlayer_0,WavePlayer_1,WavePlayer_2], properties {bShouldExcludeFromBranchCulling=true, weights [1,1,1]}}
 {id=WavePlayer_0, properties {soundWaveAssetPtr=&/GrappleKnife_Prj_Latch_And_Pull_01}}
 {id=WavePlayer_1, properties {soundWaveAssetPtr=&/GrappleKnife_Prj_Latch_And_Pull_02}}
 {id=WavePlayer_2, properties {soundWaveAssetPtr=&/GrappleKnife_Prj_Latch_And_Pull_03}}
]
properties {
 attenuationSettings=&/GrappleKnife_Prj_Latch_and_Reel_Att
 bOverrideConcurrency=true
 maxDistance=16000
 soundClassObject=&Fort_Gadgets
 volumeMultiplier=0.9
 attenuationOverrides {bEnableReverbSend=false, falloffDistance=3000, nonSpatializedRadiusStart=1000, stereoSpread=2000, attenuationShapeExtents {x=2500}}
 concurrencyOverrides {maxCount=3}
 concurrencySet [&RifleBulletImpact_Conc]
}
