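# Sound cue for the grapple knife projectile's player impact, routed to the Fort_Impacts sound class.
# A random node picks one of three equally weighted impact waves and feeds a modulator that randomly raises pitch slightly while keeping volume fixed.
# The cue boosts overall volume and uses the shared latch-and-reel attenuation asset.
# Its own concurrency override caps simultaneous instances, taking precedence over the assigned rifle bullet impact concurrency set.
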SoundCue /GrappleWeapon/Assets/Audio/GrappleKnife_Prj_Cast_Impact_Player_Cue
root=Modulator_1
nodes [
 {id=Modulator_1, children [Random_1], properties {pitchMax=1.2, pitchMin=1, volumeMax=1, volumeMin=1}}
 {id=Random_1, children [WavePlayer_3,WavePlayer_4,WavePlayer_5], properties {bShouldExcludeFromBranchCulling=true, weights [1,1,1]}}
 {id=WavePlayer_3, properties {soundWaveAssetPtr=&/GrappleKnife_Prj_Impact_Player_01}}
 {id=WavePlayer_4, properties {soundWaveAssetPtr=&/GrappleKnife_Prj_Impact_Player_02}}
 {id=WavePlayer_5, properties {soundWaveAssetPtr=&/GrappleKnife_Prj_Impact_Player_03}}
]
properties {
 attenuationSettings=&/GrappleKnife_Prj_Latch_and_Reel_Att
 bOverrideConcurrency=true
 maxDistance=16000
 soundClassObject=&Fort_Impacts
 volumeMultiplier=1.5
 attenuationOverrides {
  bAttenuate=false
  bEnableReverbSend=false
  bSpatialize=false
  falloffDistance=3000
  nonSpatializedRadiusStart=1000
  stereoSpread=2000
  attenuationShapeExtents {x=2500}
 }
 concurrencyOverrides {maxCount=3}
 concurrencySet [&RifleBulletImpact_Conc]
}
