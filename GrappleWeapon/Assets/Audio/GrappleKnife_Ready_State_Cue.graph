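# Sound cue for the grapple knife's ready-state UI blip, audible only when the IsLocal bool parameter is true.
# A Branch node on IsLocal feeds an equally weighted Random node that picks one of three blip waves, leaving the false and unset inputs empty.
# Overridden concurrency allows a single instance and stops the oldest, and virtualization is disabled.
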SoundCue /GrappleWeapon/Assets/Audio/GrappleKnife_Ready_State_Cue
root=Branch_0
nodes [
 {id=Branch_0, children [Random_1,null,null], properties {boolParameterName=IsLocal}}
 {id=Random_1, children [WavePlayer_3,WavePlayer_5,WavePlayer_6], properties {bShouldExcludeFromBranchCulling=true, weights [1,1,1]}}
 {id=WavePlayer_3, properties {soundWaveAssetPtr=&/GrappleKnife_UI_Ready_Blip_01}}
 {id=WavePlayer_5, properties {soundWaveAssetPtr=&/GrappleKnife_UI_Ready_Blip_02}}
 {id=WavePlayer_6, properties {soundWaveAssetPtr=&/GrappleKnife_UI_Ready_Blip_03}}
]
properties {
 bOverrideConcurrency=true
 maxDistance=2097152
 virtualizationMode=Disabled
 volumeMultiplier=1.5
 attenuationOverrides {falloffDistance=5000, attenuationShapeExtents {x=1000}}
 concurrencyOverrides {maxCount=1, resolutionRule=StopOldest}
}
