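# Sound cue for the grapple knife reattach sound in the GrappleWeapon item.
# A random node picks one of four equally weighted reattach waves and feeds it through a modulator for variation.
# It uses the knife foley attenuation and gadget item sound class, with a concurrency override that prevents new instances once the limit is reached.
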SoundCue /GrappleWeapon/Assets/Audio/GrappleKnife_Reattach_Cue
root=Modulator_0
nodes [
 {id=Modulator_0, children [Random_0]}
 {
  id=Random_0
  children [WavePlayer_0,WavePlayer_1,WavePlayer_2,WavePlayer_3]
  properties {bShouldExcludeFromBranchCulling=true, weights [1,1,1,1]}
 }
 {id=WavePlayer_0, properties {soundWaveAssetPtr=&/GrappleKnife_Reattached_01}}
 {id=WavePlayer_1, properties {soundWaveAssetPtr=&/GrappleKnife_Reattached_02}}
 {id=WavePlayer_2, properties {soundWaveAssetPtr=&/GrappleKnife_Reattached_03}}
 {id=WavePlayer_3, properties {soundWaveAssetPtr=&/GrappleKnife_Reattached_04}}
]
properties {
 attenuationSettings=&/GrappleKnife_Foley_Att
 bOverrideConcurrency=true
 maxDistance=6000
 soundClassObject=&Fort_Gadgets_Items
 volumeMultiplier=1
 concurrencyOverrides {resolutionRule=PreventNew}
}
