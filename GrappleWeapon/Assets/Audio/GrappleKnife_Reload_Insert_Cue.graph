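# Sound cue named for the insert step of the Grapple Knife's reload, set up as weapon foley but containing no sound nodes.
# Its graph has no root, so it has no audio source to play.
# It carries only settings: the shared Fortnite weapon reload attenuation, the weapon foley sound class, lowered volume, a concurrency override and an enabled debug flag.
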SoundCue /GrappleWeapon/Assets/Audio/GrappleKnife_Reload_Insert_Cue
root=null
nodes []
properties {
 attenuationSettings=&Fort_Weapon_Reload_Att
 bDebug=true
 bOverrideConcurrency=true
 maxDistance=2250
 soundClassObject=&Fort_WeaponFoley
 volumeMultiplier=0.501
}
