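# Sound cue for the GrappleKnife heavy swing, layering a main swing sound with a sweetener.
# A mixer blends two randomized layers: one of three heavy swing waves, and one of five sweetener waves mixed slightly louder.
# Each layer has its own modulator, with the sweetener pitched down and varying widely in volume.
# It uses close-range melee attenuation, the player melee weapon sound class and a melee concurrency limit.
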SoundCue /GrappleWeapon/Assets/Audio/GrappleKnife_Swing_Heavy_Cue
root=Mixer_0
nodes [
 {id=Mixer_0, children [Modulator_0,Modulator_1], properties {inputVolume [1,1.2]}}
 {id=Modulator_0, children [Random_2], properties {pitchMax=1.1, volumeMax=1, volumeMin=0.9}}
 {id=Modulator_1, children [Random_0], properties {pitchMax=0.95, pitchMin=0.85, volumeMax=1, volumeMin=0.1}}
 {
  id=Random_0
  children [WavePlayer_0,WavePlayer_1,WavePlayer_2,WavePlayer_3,WavePlayer_4]
  properties {bShouldExcludeFromBranchCulling=true, weights [1,1,1,1,1]}
 }
 {id=Random_2, children [WavePlayer_6,WavePlayer_8,WavePlayer_9], properties {bShouldExcludeFromBranchCulling=true, weights [1,1,1]}}
 {id=WavePlayer_0, properties {soundWaveAssetPtr=&/GrappleKnife_Swing_Sweetener_01}}
 {id=WavePlayer_1, properties {soundWaveAssetPtr=&/GrappleKnife_Swing_Sweetener_02}}
 {id=WavePlayer_2, properties {soundWaveAssetPtr=&/GrappleKnife_Swing_Sweetener_03}}
 {id=WavePlayer_3, properties {soundWaveAssetPtr=&/GrappleKnife_Swing_Sweetener_04}}
 {id=WavePlayer_4, properties {soundWaveAssetPtr=&/GrappleKnife_Swing_Sweetener_05}}
 {id=WavePlayer_6, properties {soundWaveAssetPtr=&/GrappleKnife_Swing_Heavy_01}}
 {id=WavePlayer_8, properties {soundWaveAssetPtr=&/GrappleKnife_Swing_Heavy_02}}
 {id=WavePlayer_9, properties {soundWaveAssetPtr=&/GrappleKnife_Swing_Heavy_03}}
]
properties {
 attenuationSettings=/Game/Sounds/Attenuation/Fort_Weapon_Melee_Attack_Close_Range_Stereo_Att
 maxDistance=10058.75
 soundClassObject=&Fort_WeaponMelee_PSM
 volumeMultiplier=1
 concurrencySet [&Player_Melee_Weapon_Conc]
}
