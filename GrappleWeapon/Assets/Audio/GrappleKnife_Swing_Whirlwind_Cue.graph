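# Sound Cue for the grapple knife's whirlwind swing in the GrappleWeapon content.
# A Random node picks among three equally weighted air-attack twirl waves and is excluded from branch culling.
# The pick feeds a root Modulator that holds volume constant.
# The cue uses melee special-dash attenuation, a melee weapon sound class and the player melee weapon concurrency set.
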SoundCue /GrappleWeapon/Assets/Audio/GrappleKnife_Swing_Whirlwind_Cue
root=Modulator_0
nodes [
 {id=Modulator_0, children [Random_0], properties {volumeMax=1, volumeMin=1}}
 {id=Random_0, children [WavePlayer_1,WavePlayer_2,WavePlayer_3], properties {bShouldExcludeFromBranchCulling=true, weights [1,1,1]}}
 {id=WavePlayer_1, properties {soundWaveAssetPtr=&/GrappleKnife_AirAttack_Twirl_01}}
 {id=WavePlayer_2, properties {soundWaveAssetPtr=&/GrappleKnife_AirAttack_Twirl_02}}
 {id=WavePlayer_3, properties {soundWaveAssetPtr=&/GrappleKnife_AirAttack_Twirl_03}}
]
properties {
 attenuationSettings=/Game/Sounds/Attenuation/Fort_Weapon_Melee_Attack_Special_Dash_Att
 maxDistance=10058.75
 soundClassObject=&Fort_WeaponMelee_PSM
 volumeMultiplier=1
 concurrencySet [&Player_Melee_Weapon_Conc]
}
