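# Sound cue for the short twirl of the grapple knife, a jingle foley sound from the grapple weapon.
# A Random node picks one of six equally weighted jingle foley waves and is excluded from branch culling.
# A Modulator above it varies only the volume on each play, so repeated twirls sound less uniform.
# The cue uses the grapple knife attenuation asset and the gadgets and items sound class, at reduced overall volume.
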SoundCue /GrappleWeapon/Assets/Audio/GrappleKnife_Twirl_Short_Cue
root=Modulator_1
nodes [
 {id=Modulator_1, children [Random_1], properties {volumeMax=1, volumeMin=0.45}}
 {
  id=Random_1
  children [WavePlayer_4,WavePlayer_5,WavePlayer_6,WavePlayer_7,WavePlayer_8,WavePlayer_9]
  properties {bShouldExcludeFromBranchCulling=true, weights [1,1,1,1,1,1]}
 }
 {id=WavePlayer_4, properties {soundWaveAssetPtr=&/GrappleKnife_Jingle_Foley_01}}
 {id=WavePlayer_5, properties {soundWaveAssetPtr=&/GrappleKnife_Jingle_Foley_02}}
 {id=WavePlayer_6, properties {soundWaveAssetPtr=&/GrappleKnife_Jingle_Foley_03}}
 {id=WavePlayer_7, properties {soundWaveAssetPtr=&/GrappleKnife_Jingle_Foley_04}}
 {id=WavePlayer_8, properties {soundWaveAssetPtr=&/GrappleKnife_Jingle_Foley_05}}
 {id=WavePlayer_9, properties {soundWaveAssetPtr=&/GrappleKnife_Jingle_Foley_06}}
]
properties {
 attenuationSettings=&/GrappleKnife_Att
 maxDistance=10000
 soundClassObject=&Fort_Gadgets_Items
 volumeMultiplier=0.6
}
