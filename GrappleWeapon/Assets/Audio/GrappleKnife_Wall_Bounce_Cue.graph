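# Sound cue for the grapple knife bouncing off walls, picking randomly among five bounce waves with slight pitch modulation.
# An IsLocal branch plays local or unset cases at full level, while remote cases pass through an IsEnemy branch.
# That branch keeps enemy or unset sounds at full volume and mixes friendly ones down to a quarter.
# All paths are attenuated, and a concurrency override caps simultaneous instances.
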SoundCue /GrappleWeapon/Assets/Audio/GrappleKnife_Wall_Bounce_Cue
root=Branch_0
nodes [
 {id=Attenuation_0, children [Branch_1], properties {attenuationSettings=&/GrappleKnife_Wall_Bounce_Att}}
 {id=Attenuation_1, children [Modulator_0], properties {attenuationSettings=&GrappleKnife_Wall_Bounce_Att}}
 {id=Branch_0, children [Attenuation_1, Attenuation_0, Attenuation_1], properties {boolParameterName=IsLocal}}
 {id=Branch_1, children [Mixer_0,Mixer_1,Mixer_0], properties {boolParameterName=IsEnemy}}
 {id=Mixer_0, children [Modulator_1], properties {inputVolume [1]}}
 {id=Mixer_1, children [Modulator_2], properties {inputVolume [0.25]}}
 {id=Modulator_0, children [Random_0], properties {pitchMax=1.1, volumeMax=1}}
 {id=Modulator_1, children [Random_1], properties {pitchMax=1.1, volumeMax=1}}
 {id=Modulator_2, children [Random_2], properties {pitchMax=1.1, volumeMax=1}}
 {
  id=Random_0
  children [WavePlayer_0,WavePlayer_1,WavePlayer_2,WavePlayer_4,WavePlayer_5]
  properties {bShouldExcludeFromBranchCulling=true, weights [1,1,1,1,1]}
 }
 {
  id=Random_1
  children [WavePlayer_6, WavePlayer_7, WavePlayer_8, WavePlayer_9, WavePlayer_10]
  properties {bShouldExcludeFromBranchCulling=true, weights [1,1,1,1,1]}
 }
 {
  id=Random_2
  children [WavePlayer_11, WavePlayer_12, WavePlayer_13, WavePlayer_14, WavePlayer_15]
  properties {bShouldExcludeFromBranchCulling=true, weights [1,1,1,1,1]}
 }
 {id=WavePlayer_0, properties {soundWaveAssetPtr=&/GrappleKnife_Wall_Bounce_01}}
 {id=WavePlayer_1, properties {soundWaveAssetPtr=&/GrappleKnife_Wall_Bounce_02}}
 {id=WavePlayer_10, properties {soundWaveAssetPtr=&/GrappleKnife_Wall_Bounce_05}}
 {id=WavePlayer_11, properties {soundWaveAssetPtr=&GrappleKnife_Wall_Bounce_01}}
 {id=WavePlayer_12, properties {soundWaveAssetPtr=&GrappleKnife_Wall_Bounce_02}}
 {id=WavePlayer_13, properties {soundWaveAssetPtr=&/GrappleKnife_Wall_Bounce_03}}
 {id=WavePlayer_14, properties {soundWaveAssetPtr=&/GrappleKnife_Wall_Bounce_04}}
 {id=WavePlayer_15, properties {soundWaveAssetPtr=&GrappleKnife_Wall_Bounce_05}}
 {id=WavePlayer_2, properties {soundWaveAssetPtr=&GrappleKnife_Wall_Bounce_03}}
 {id=WavePlayer_4, properties {soundWaveAssetPtr=&GrappleKnife_Wall_Bounce_04}}
 {id=WavePlayer_5, properties {soundWaveAssetPtr=&GrappleKnife_Wall_Bounce_05}}
 {id=WavePlayer_6, properties {soundWaveAssetPtr=&GrappleKnife_Wall_Bounce_01}}
 {id=WavePlayer_7, properties {soundWaveAssetPtr=&GrappleKnife_Wall_Bounce_02}}
 {id=WavePlayer_8, properties {soundWaveAssetPtr=&GrappleKnife_Wall_Bounce_03}}
 {id=WavePlayer_9, properties {soundWaveAssetPtr=&GrappleKnife_Wall_Bounce_04}}
]
properties {
 bOverrideConcurrency=true
 maxDistance=6000
 soundClassObject=&Fort_Gadgets_Items
 volumeMultiplier=1
 concurrencyOverrides {maxCount=3}
}
