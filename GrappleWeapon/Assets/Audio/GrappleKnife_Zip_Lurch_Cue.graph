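# Sound cue for the grapple knife weapon's zip lurch, playing a wind gust as the player is pulled along.
# A Random node picks among four equally weighted grappled wind gust waves, feeding a Modulator that slightly varies pitch on each play.
# The cue is boosted in volume, routed to the gadgets and items sound class, and spatialized with the knife's foley attenuation.
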SoundCue /GrappleWeapon/Assets/Audio/GrappleKnife_Zip_Lurch_Cue
root=Modulator_1
nodes [
 {id=Modulator_1, children [Random_1], properties {pitchMax=1.1, pitchMin=0.9}}
 {id=Random_1, children [WavePlayer_4,WavePlayer_5,WavePlayer_6,WavePlayer_7], properties {weights [1,1,1,1]}}
 {id=WavePlayer_4, properties {soundWaveAssetPtr=&/GrappleKnife_Grappled_Wind_Gust_01}}
 {id=WavePlayer_5, properties {soundWaveAssetPtr=&/GrappleKnife_Grappled_Wind_Gust_02}}
 {id=WavePlayer_6, properties {soundWaveAssetPtr=&/GrappleKnife_Grappled_Wind_Gust_03}}
 {id=WavePlayer_7, properties {soundWaveAssetPtr=&/GrappleKnife_Grappled_Wind_Gust_04}}
]
properties {
 attenuationSettings=&/GrappleKnife_Foley_Att
 maxDistance=6000
 soundClassObject=&Fort_Gadgets_Items
 volumeMultiplier=1.5
}
